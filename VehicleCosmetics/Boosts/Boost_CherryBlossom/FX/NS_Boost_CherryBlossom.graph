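# Niagara system for the Cherry Blossom vehicle boost cosmetic, trailing pink petals and glowing ribbons from the car's boost sockets.
# Flowers and leafs emitters partition sprites across skeletal mesh sockets, throwing them backward with spin and vortex forces; only Flowers animates SubUVs.
# Ribbons spawns quality-scaled trails per port, computing four sine-driven "noodle" positions and bloom values alongside four ribbon renderers.
# Painted-variant override modules are present but disabled.
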NiagaraSystem /VehicleCosmetics/Boosts/Boost_CherryBlossom/FX/NS_Boost_CherryBlossom
dependencies {
 usedDataInterfaces [
  &NiagaraDataInterfaceCurve
  &NiagaraDataInterfaceSpriteRendererInfo
  &NiagaraDataInterfaceSkeletalMesh
 ]
 usedDynamicInputs []
 usedModules [
  &self:SystemSpawnScript.SystemScriptSource.SystemScriptGraph.NiagaraNodeAssignment_0.SetVariables_#1
  &SystemState
 ]
}
emitters {
 Flowers {
  emitterUpdateScript [
   &EmitterState()
   &VC_FilteredSocketCount()
   &SpawnRate(SpawnRate=100, Use Spawn Probability=false)
  ]
  particleSpawnScript [
   &VC_PaintedVariantsOverride() {enabled=false}
   &InitializeParticle(
    Color Maximum=<rgba 1 0.2 0.5 1>
    Color Minimum=<rgba 1 0.5 0.7 1>
    Color Mode=::NewEnumerator2
    Lifetime=&RandomRangeFloat(Minimum=1, Maximum=1.5)
    Sprite Rotation Mode=::NewEnumerator2
    Sprite Size Mode=::NewEnumerator3
    Uniform Sprite Size=&RandomRangeFloat(Minimum=15, Maximum=25)
    UsePositionOffset=false
    Write Material Random=true
   )
   &PartitionParticles(Number Of Partitions=$Emitter.PortCount)
   &SkeletalMeshLocation(
    Bone / Socket Index=$Particles.PartitionParticles.Partition
    Bone / Socket Sampling Mode=::NewEnumerator5
    MaskSpawnGroup=false
    Mesh Sampling Type=::NewEnumerator6
    Sampled Position Offset=&AddVectorToPosition(Position=<0 0 0>, Vector=$User.DTCarOffset)
    Skeletal Mesh=$System.SkeletalMesh
   )
   &ShapeLocation(MaskSpawnGroup=false, Override Seed=false, Sphere Radius=12, Sphere Surface Distribution=10, UseOwnerScale=true)
   &AddVelocity(Velocity=&RandomRangeVector(Minimum=<-600 70 120>, Maximum=<-600 -140 0>))
  ]
  particleUpdateScript [
   &ParticleState()
   &SpriteRotationRate(Rotation Rate=&RandomRangeFloat(Minimum=-360, Maximum=360))
   &ScaleColor(
    Scale Alpha=&FloatFromCurve(FloatCurve=DataInterface() {propertyValues {Curve {keys [<at 0 1 RCIM_Cubic>,<at 1.0011 0.002 RCIM_Cubic>]}}})
    Scale RGB=<6 6 6>
    ScaleA=true
    ScaleRGB=true
   )
   &VortexForce(
    UseOriginOffset=false
    Vortex Axis=&RandomRangeVector(Minimum=<-2 0 0>, Maximum=<2 0 0>)
    Vortex Force Amount=100
   )
   &SolveForcesAndVelocity(Clamp Velocity=false, Limit Acceleration=false)
   &SubUVAnimation(
    "Sprite Renderer "=DataInterface() {propertyValues {SpriteRenderer=&self:Flowers_0.NiagaraSpriteRendererProperties_3}}
    UseEndFrame=false
    UseStartFrame=false
   )
  ]
  rendererClasses [&NiagaraSpriteRendererProperties]
  renderers [{rendererClass=&NiagaraSpriteRendererProperties}]
 }
 Ribbons {
  emitterUpdateScript [
   &EmitterState(Life Cycle Mode=::NewEnumerator1, Loop Duration=1, UseLoopDelay=false)
   &VC_FilteredSocketCount()
   &VC_RateQualityHandler(High Spawn Rate=50, Low Spawn Rate=30, Medium Spawn Rate=40, Use Custom Rate?=true)
   &VC_SpawnRibbon_v1(Number of Ribbons=$Emitter.PortCount, SpawnRate=$Emitter.Spawn Rate)
  ]
  particleSpawnScript [
   &VC_PaintedVariantsOverride() {enabled=false}
   &VC_InitializeRibbon_v1(
    Color=<rgba 1.5 3 2.4 1>
    Custom Offset=true
    Enable Skeletal Position=true
    Lifetime=0.8
    Local Vector Offset=&AddVectorToPosition(Position=<50 0 0>, Vector=$User.DTCarOffset)
    Offset Amount=1
    Ribbon Width=32
   )
   &AddVelocity(Velocity=<-2000 0 0>) {module=AddVelocity001}
  ]
  particleUpdateScript [
   &ParticleState()
   &ScaleColor(
    Scale Alpha=&FloatFromCurve(FloatCurve=DataInterface() {propertyValues {Curve {keys [
         {tangentMode=RCTM_User, time=0, value=0}
         {arriveTangent=-1.11, interpMode=RCIM_Cubic, leaveTangent=-1.11, tangentMode=RCTM_User, time=0.15, value=1}
         <at 1 0 RCIM_Cubic>
        ]}}})
    Scale RGB=&VectorFromFloat(Value=&FloatFromCurve(FloatCurve=DataInterface() {propertyValues {Curve {keys [{tangentMode=RCTM_User, time=0, value=0}, {tangentMode=RCTM_User, time=0.15, value=0.5}, <at 1 0.2 RCIM_Cubic>]}}}))
    ScaleA=true
    ScaleRGB=true
   )
   &SolveForcesAndVelocity(Clamp Velocity=false, Limit Acceleration=false)
   &VC_UpdateRibbon_v1(Enable Skeletal Position=true)
   &ScaleRibbonWidth(Ribbon Width Scale=&FloatFromCurve(FloatCurve=DataInterface() {propertyValues {Curve {keys [<at 0 3>,<at 0.9 0.625>,<at 1 0>]}}}))
   /SetVariables(
    Particles.PosNoodle1:NiagaraPosition=&AddVectorToPosition(
     Position=$Particles.Position
     Vector=&TransformVector(
      Vector=&MakeVector(
       X=0
       Y=&Sine(
        Normalized Angle=&Add_Float(
         A=&Multiply_Float(A=$Particles.NormalizedAge, B=-1.3)
         B=&Multiply_Float(A=$Emitter.LoopedAge, B=1)
        )
        Period=&MaskFloatBySpawnGroup(Masked Float=0.3, Passthrough Float=-0.3)
        Scale=&FloatFromCurve(
         FloatCurve=DataInterface() {propertyValues {Curve {keys [{arriveTangent=8.1681, interpMode=RCIM_Cubic, leaveTangent=8.1681, tangentMode=RCTM_User, time=0, value=0}, <at 0.2717 1>, <at 1 1.5>]}}}
         Scale Curve=5.5
        )
        Bias=0
       )
       Z=&Sine(
        Normalized Angle=&Add_Float(
         A=&Multiply_Float(A=$Particles.NormalizedAge, B=-1.3)
         B=&Multiply_Float(A=$Emitter.LoopedAge, B=1)
        )
        Period=0.3
        Scale=&FloatFromCurve(
         FloatCurve=DataInterface() {propertyValues {Curve {keys [{arriveTangent=8.1681, interpMode=RCIM_Cubic, leaveTangent=8.1681, tangentMode=RCTM_User, time=0, value=0}, <at 0.2717 1>, <at 1 1.5>]}}}
         Scale Curve=5
        )
        Bias=0
       )
      )
      Destination Space=::Simulation
      Apply Scale=true
     )
    )
    Particles.PosNoodle2:NiagaraPosition=&AddVectorToPosition(
     Position=&AddVectorToPosition(
      Position=$Particles.Position
      Vector=&TransformVector(
       Vector=&MakeVector(
        X=-5
        Y=&Sine(
         Normalized Angle=&Add_Float(
          A=&Multiply_Float(A=$Particles.NormalizedAge, B=-1.3)
          B=&Multiply_Float(A=$Emitter.LoopedAge, B=1)
         )
         Period=&MaskFloatBySpawnGroup(Masked Float=0.25, Passthrough Float=-0.25)
         Scale=&FloatFromCurve(
          FloatCurve=DataInterface() {propertyValues {Curve {keys [{arriveTangent=8.1681, interpMode=RCIM_Cubic, leaveTangent=8.1681, tangentMode=RCTM_User, time=0, value=0}, <at 0.2717 1>, <at 1 1.5>]}}}
          Scale Curve=7
         )
         Bias=0
        )
        Z=&Sine(
         Normalized Angle=&Add_Float(
          A=&Multiply_Float(A=$Particles.NormalizedAge, B=-1.3)
          B=&Multiply_Float(A=$Emitter.LoopedAge, B=1)
         )
         Period=0.4
         Scale=&FloatFromCurve(
          FloatCurve=DataInterface() {propertyValues {Curve {keys [{arriveTangent=8.1681, interpMode=RCIM_Cubic, leaveTangent=8.1681, tangentMode=RCTM_User, time=0, value=0}, <at 0.2 1.6>, <at 1 1.05>]}}}
          Scale Curve=5
         )
         Bias=0
        )
       )
       Destination Space=::Simulation
       Apply Scale=true
      )
     )
     Vector=<0 -2.5 -2.5>
    )
    Particles.PosNoodle3:NiagaraPosition=&AddVectorToPosition(
     Position=&AddVectorToPosition(
      Position=$Particles.Position
      Vector=&TransformVector(
       Vector=&MakeVector(
        X=-10
        Y=&Sine(
         Normalized Angle=&Add_Float(
          A=&Multiply_Float(A=$Particles.NormalizedAge, B=-1.3)
          B=&Multiply_Float(A=$Emitter.LoopedAge, B=1)
         )
         Period=&MaskFloatBySpawnGroup(Masked Float=0.35, Passthrough Float=-0.35)
         Scale=&FloatFromCurve(
          FloatCurve=DataInterface() {propertyValues {Curve {keys [{arriveTangent=8.1681, interpMode=RCIM_Cubic, leaveTangent=8.1681, tangentMode=RCTM_User, time=0, value=0}, <at 0.3 1.4>, <at 1 0.95>]}}}
          Scale Curve=7
         )
         Bias=0
        )
        Z=&Sine(
         Normalized Angle=&Add_Float(
          A=&Multiply_Float(A=$Particles.NormalizedAge, B=-1.3)
          B=&Multiply_Float(A=$Emitter.LoopedAge, B=1)
         )
         Period=-0.5
         Scale=&FloatFromCurve(
          FloatCurve=DataInterface() {propertyValues {Curve {keys [{arriveTangent=8.1681, interpMode=RCIM_Cubic, leaveTangent=8.1681, tangentMode=RCTM_User, time=0, value=0}, <at 0.2717 1>, <at 1 1.5>]}}}
          Scale Curve=5
         )
         Bias=0
        )
       )
       Destination Space=::Simulation
       Apply Scale=true
      )
     )
     Vector=<0 2.5 2.5>
    )
    Particles.PosNoodle4:NiagaraPosition=&AddVectorToPosition(
     Position=$Particles.Position
     Vector=&TransformVector(
      Vector=&Add_Vector(
       A=&MakeVector(
        X=-25
        Y=&Sine(
         Normalized Angle=&Add_Float(
          A=&Multiply_Float(A=$Particles.NormalizedAge, B=-1.3)
          B=&Multiply_Float(A=$Emitter.LoopedAge, B=1)
         )
         Period=&MaskFloatBySpawnGroup(Masked Float=0.6, Passthrough Float=-0.6)
         Scale=&FloatFromCurve(FloatCurve=DataInterface() {propertyValues {Curve {keys [
              {arriveTangent=0.1945, interpMode=RCIM_Cubic, leaveTangent=0.1945, tangentMode=RCTM_User, time=0, value=0}
              <at 0.2717 1>
              {arriveTangent=-34.9295, interpMode=RCIM_Cubic, leaveTangent=-34.9295, tangentMode=RCTM_User, time=0.8, value=5}
              <at 0.9 0>
             ]}}})
         Bias=0
        )
        Z=&Sine(
         Normalized Angle=&Add_Float(
          A=&Multiply_Float(A=$Particles.NormalizedAge, B=-1.3)
          B=&Multiply_Float(A=$Emitter.LoopedAge, B=1)
         )
         Period=0.5
         Scale=&FloatFromCurve(
          FloatCurve=DataInterface() {propertyValues {Curve {keys [
              <at 0 0 RCIM_Cubic>
              <at 0.2717 1>
              {arriveTangent=-49.0194, interpMode=RCIM_Cubic, leaveTangent=-49.0194, tangentMode=RCTM_User, time=0.8, value=5}
              <at 0.9 0>
             ]}}}
          Scale Curve=-1
         )
         Bias=0
        )
       )
       B=<0 -7 -3>
      )
      Destination Space=::Simulation
      Apply Scale=true
     )
    )
   ) {isSetParameters=true}
   /SetVariables(
    Particles.BloomColor:LinearColor=&MultiplyLinearColor(Linear Color=$Particles.Color, Scale Factor=<1 1 1 0.2>)
    Particles.BloomSize:NiagaraFloat=&Multiply_Float(
     A=60
     B=&FloatFromCurve(FloatCurve=DataInterface() {propertyValues {Curve {keys [
          {arriveTangent=-2.5551, interpMode=RCIM_Cubic, leaveTangent=-2.5551, tangentMode=RCTM_User, time=0, value=3}
          {arriveTangent=-1.1193, interpMode=RCIM_Cubic, leaveTangent=-1.1193, tangentMode=RCTM_User, time=0.7, value=1.75}
          {arriveTangent=-12.4055, interpMode=RCIM_Cubic, leaveTangent=-12.4055, tangentMode=RCTM_User, time=0.98, value=0}
         ]}}})
    )
   ) {isSetParameters=true}
  ]
  rendererClasses [&NiagaraRibbonRendererProperties]
  renderers [
   {rendererClass=&NiagaraRibbonRendererProperties}
   {rendererClass=&NiagaraRibbonRendererProperties, rendererIndex=1}
   {rendererClass=&NiagaraRibbonRendererProperties, rendererIndex=2}
   {rendererClass=&NiagaraRibbonRendererProperties, rendererIndex=3}
  ]
 }
 leafs {
  emitterUpdateScript [
   &EmitterState()
   &VC_FilteredSocketCount()
   &SpawnRate(SpawnRate=50, Use Spawn Probability=false)
  ]
  particleSpawnScript [
   &VC_PaintedVariantsOverride() {enabled=false}
   &InitializeParticle(
    Color Maximum=<rgba 1 0.2 0.5 1>
    Color Minimum=<rgba 1 0.5 0.7 1>
    Color Mode=::NewEnumerator2
    Lifetime=&RandomRangeFloat(Minimum=1, Maximum=1.5)
    Sprite Rotation Mode=::NewEnumerator2
    Sprite Size Mode=::NewEnumerator3
    Uniform Sprite Size=&RandomRangeFloat(Minimum=5, Maximum=10)
    UsePositionOffset=false
    Write Material Random=true
   )
   &PartitionParticles(Number Of Partitions=$Emitter.PortCount)
   &SkeletalMeshLocation(
    Bone / Socket Index=$Particles.PartitionParticles.Partition
    Bone / Socket Sampling Mode=::NewEnumerator5
    MaskSpawnGroup=false
    Mesh Sampling Type=::NewEnumerator6
    Sampled Position Offset=&AddVectorToPosition(Position=<0 0 0>, Vector=$User.DTCarOffset)
    Skeletal Mesh=$System.SkeletalMesh
   )
   &ShapeLocation(MaskSpawnGroup=false, Override Seed=false, Sphere Radius=12, Sphere Surface Distribution=10, UseOwnerScale=true)
   &AddVelocity(Velocity=&RandomRangeVector(Minimum=<-600 70 120>, Maximum=<-600 -140 0>))
  ]
  particleUpdateScript [
   &ParticleState()
   &SpriteRotationRate(Rotation Rate=&RandomRangeFloat(Minimum=-360, Maximum=360))
   &ScaleColor(
    Scale Alpha=&FloatFromCurve(FloatCurve=DataInterface() {propertyValues {Curve {keys [<at 0 1 RCIM_Cubic>,<at 1.0011 0.002 RCIM_Cubic>]}}})
    Scale RGB=<6 6 6>
    ScaleA=true
    ScaleRGB=true
   )
   &VortexForce(
    UseOriginOffset=false
    Vortex Axis=&RandomRangeVector(Minimum=<-2 0 0>, Maximum=<2 0 0>)
    Vortex Force Amount=100
   )
   &SolveForcesAndVelocity(Clamp Velocity=false, Limit Acceleration=false)
  ]
  rendererClasses [&NiagaraSpriteRendererProperties]
  renderers [{rendererClass=&NiagaraSpriteRendererProperties}]
 }
}
summary {
 systemName=NS_Boost_CherryBlossom
 emitters [
  {emitterName=Ribbons, rendererClasses [&NiagaraRibbonRendererProperties]}
  {emitterName=Flowers, rendererClasses [&NiagaraSpriteRendererProperties]}
  {emitterName=leafs, rendererClasses [&NiagaraSpriteRendererProperties]}
 ]
 userVariables [
  User.bIsFrontendPreview {type=NiagaraBool}
  User.bIsPainted {type=NiagaraBool}
  User.ColorVariantID {type=NiagaraInt32}
  User.DTCarOffset {type=Vector3f, description="", defaultValue {struct=Vector3f, value=<0 0 0>}}
  User.Intensity {type=NiagaraFloat, defaultValue=1, description=""}
  User.LightningBrightness {type=NiagaraFloat, defaultValue=1, description=""}
  User.LightningBrightnessRed {type=NiagaraFloat, defaultValue=1, description=""}
  User.PaintedColor {type=LinearColor, description="", defaultValue {struct=LinearColor, value=<rgba 1 1 1 1>}}
  User.PaintedScalar {type=NiagaraFloat, defaultValue=1, description=""}
  User.PortCount {type=NiagaraInt32, defaultValue=2}
 ]
}
